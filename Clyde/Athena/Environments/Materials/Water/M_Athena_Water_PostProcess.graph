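# Athena water post-process material that renders an underwater view against a flat virtual plane at a WaterLevel height.
# It intersects view rays with that plane and compares scene depth to get a surface mask and distance through water.
# Two panning normal maps drive masked refraction and a Fresnel blend to a reflection color.
# Exponential per-channel extinction and distance fog tint the scene, disabled where camera and pixel are both above water.
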Material /Clyde/Athena/Environments/Materials/Water/M_Athena_Water_PostProcess
graph {
 properties {blendableLocation=BL_SceneColorAfterDOF, materialDomain=MD_PostProcess, pixelDepthOffsetMode=PDOM_Legacy, usedShadingModels=Default Lit}
 nodes [
  n1 = ScreenPosition()
  n2 = TextureCoordinate()
  n3 = CameraPositionWS()
  n4 = CameraPositionWS()
  n5 = CameraVectorWS()
  n6 = ComponentMask(None=@n5) {b=true}
  n7 = Multiply(
   A=Subtract(A=ScalarParameter() {defaultValue=-48.5, parameterName=WaterLevel}, B=ComponentMask(None=@n4) {b=true})
   B=Divide(A=@n5, B=Min(A=@n6) {constB=0})
  ) {desc=Vec to Waterplane}
  n8 = ComponentMask(None=Divide(A=Add(A=@n3, B=@n7), B=ScalarParameter() {defaultValue=128, parameterName=Water UV Tiling})) {g=true, r=true}
  n9 = Time()
  n10 = ScalarParameter() {defaultValue=0.0905, parameterName=Speed1}
  n11 = Divide(A=@n8) {constB=1.618}
  n12 = Time()
  n13 = ScalarParameter() {defaultValue=-0.1827, parameterName=Speed2}
  n14 = Add(
   A=TextureSample(UVs=Panner(Coordinate=@n8, Time=@n9, Speed=@n10)) {samplerType=SAMPLERTYPE_Normal, texture=/Game/Packages/DS_Fortnite_TRV_Terrain/Textures/T_TRV_HF_WaterNormal}.RGB
   B=TextureSample(UVs=Panner(Coordinate=@n11, Time=@n12, Speed=@n13)) {samplerType=SAMPLERTYPE_Normal, texture=&T_TRV_HF_WaterNormal}.RGB
  )
  n15 = ScalarParameter() {defaultValue=0.02, parameterName=Refraction}
  n16 = ScreenPosition()
  n17 = TextureCoordinate()
  n18 = FeatureLevelSwitch(Default=@n16.ViewportUV, ES2=@n17, ES3_1=@n17, SM4_REMOVED=@n16.ViewportUV)
  n19 = CameraVectorWS()
  n20 = MaterialFunctionCall() {materialFunction=&CameraDirectionVector}
  n21 = Divide(A=ComponentMask(None=SceneTexture(UVs=@n18) {sceneTextureId=PPI_SceneDepth}.Color) {r=true}, B=Abs(None=DotProduct(A=@n19, B=@n20.Result)))
  n22 = Distance(A=Constant(), B=@n7) {desc=Dist to Waterplane}
  n23 = Subtract(A=@n21, B=@n22)
  n24 = Subtract(A=ComponentMask(None=WorldPosition().XYZ) {b=true}, B=ScalarParameter() {defaultValue=-48.5, parameterName=WaterLevel})
  n25 = Subtract(A=ComponentMask(None=CameraPositionWS()) {b=true}, B=ScalarParameter() {defaultValue=-48.5, parameterName=WaterLevel})
  n26 = Multiply(A=Saturate(None=Ceil(None=@n24)), B=Saturate(None=Ceil(None=@n25))) {desc=Underwater Mask}
  n27 = LinearInterpolate(A=Saturate(None=Ceil(None=@n23)), Alpha=@n26) {constB=0, desc=Mask for Upper Plane}
  n28 = Add(
   A=FeatureLevelSwitch(Default=@n1.ViewportUV, ES2=@n2, ES3_1=@n2, SM4_REMOVED=@n1.ViewportUV)
   B=Multiply(A=ComponentMask(None=@n14) {g=true, r=true}, B=Multiply(A=@n15, B=@n27))
  )
  n29 = ScalarParameter() {defaultValue=0.5, parameterName=Boost}
  n30 = Multiply(
   A=ComponentMask(None=SceneTexture(UVs=@n28) {sceneTextureId=PPI_PostProcessInput0}.Color) {b=true, g=true, r=true}
   B=Add(A=Multiply(A=@n29, B=@n27))
  )
  n31 = Multiply(A=Min(A=@n22, B=@n21) {desc=Ray Distance thru Water})
  n32 = VectorParameter() {parameterName=EctinctionColor, defaultValue=<rgba 2 15 75 0>}
  n33 = ScalarParameter() {defaultValue=25, parameterName=ExtinctionDist}
  MaterialExpressionAppendVector_2 = AppendVector(A=Divide(A=@n31, B=Multiply(A=@n32.RGB, B=@n33)), B=Divide(A=@n31, B=ScalarParameter() {defaultValue=1000, parameterName=Fog_Range}))
  n34 = Custom!(x=@MaterialExpressionAppendVector_2) {
   code="exp(-x);"
   desc="RGB = Extinction, A=Fog"
   description=exp
   outputType=CMOT_Float4
   inputs [{inputName=x, input {expression=&self:MaterialExpressionAppendVector_2}}]
  }
  n35 = LinearInterpolate(A=ComponentMask(None=@n34) {b=true, g=true, r=true}, Alpha=@n26) {constA=1}
  n36 = Multiply(
   A=MaterialFunctionCall(Normal=@n14, Flatness=ScalarParameter() {defaultValue=0.8, parameterName=Flatness}) {materialFunction=&FlattenNormal}.Result
   B=Constant3Vector() {constant=<rgba 1 1 -1 0>}
  )
  n37 = CameraVectorWS()
  n38 = Power(Base=Max(A=DotProduct(A=@n36, B=@n37)) {constB=0}, Exp=ScalarParameter() {defaultValue=1.5, parameterName=FresnelPower})
  n39 = Saturate(None=ComponentMask(None=@n34) {a=true})
  LinearInterpolate(
   A=LinearInterpolate(
    A=VectorParameter() {parameterName=Reflect, defaultValue=<rgba 0.003 0.0759 0.1 0>}.RGB
    B=Multiply(A=@n30, B=@n35)
    Alpha=LinearInterpolate(B=@n38, Alpha=@n27) {constA=1}
   )
   B=VectorParameter() {parameterName=WaterFogColor_1, defaultValue=<rgba 0.0212 0.0682 0.1 0>}.RGB
   Alpha=LinearInterpolate(A=OneMinus(None=@n39), Alpha=@n26) {constB=0}
  )
  n40 = Constant()
  MaterialFunctionCall!(
   Plane Center=MaterialFunctionCall(X=@n40, Y=@n40, Z=ScalarParameter() {defaultValue=-48.5, parameterName=WaterLevel}) {materialFunction=&MakeFloat3}.Result
   UV Scale
  ) {materialFunction=/Game/Athena/Environments/Materials/Water/VirtualPlaneCoordinates_FNBR}
  SceneDepth()
 ]
}
