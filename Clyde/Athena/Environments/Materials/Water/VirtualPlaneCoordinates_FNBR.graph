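# Material function that gives virtual coordinates, in world units, for a horizontal plane raytraced per pixel from the camera.
# The camera ray intersects the plane at Plane Center's Z height, yielding world position and center-relative position.
# A mask marks hits in front of the camera.
# UVs take the centered XY (zero on a miss), divide by UV Scale and offset by 0.5.
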MaterialFunction /Clyde/Athena/Environments/Materials/Water/VirtualPlaneCoordinates_FNBR
graph {
 properties {description="Virtual coordinates for a raytraced plane, in world units", libraryCategoriesText [Misc,Coordinates]}
 nodes [
  n1 = ComponentMask(None=CameraPositionWS() {desc=Ray Origin}) {b=true}
  n2 = FunctionInput() {
   bUsePreviewValueAsDefault=true
   desc=Plane Origin
   description=The center location of the virtual Plane
   inputName=Plane Center
   sortPriority=2
  }
  n3 = ComponentMask(None=@n2) {b=true}
  n4 = Multiply(A=CameraVectorWS()) {constB=-1}
  n5 = Divide(A=Multiply(A=Subtract(A=@n1, B=@n3) {desc=Plane Distance}) {constB=-1}, B=ComponentMask(None=@n4) {b=true}) {desc=Intersection Time1}
  n6 = Constant()
  FunctionOutput(None=If(A=@n5, B=Constant(), "A > B"=Constant() {r=1}, "A == B"=@n6, "A < B"=@n6)) {description=A 0 or 1 mask for whether the ray hits the plane., outputName=Mask, sortPriority=5}
  n7 = Add(A=CameraPositionWS() {desc=Ray Origin}, B=Multiply(A=@n5, B=@n4)) {desc=Intersection Position}
  n8 = FunctionOutput(None=Subtract(A=@n7, B=@n2)) {
   description="The world position centered on the specified position. This means position will be 0,0,0 at the specified center."
   outputName=World Position - Centered
   sortPriority=2
  }
  n9 = Constant3Vector() {constant=<rgba 1 0 0 0>}
  n10 = Constant3Vector() {constant=<rgba 0 1 0 0>}
  n11 = Constant()
  n12 = If(A=@n5, B=Constant(), "A > B"=AppendVector(A=DotProduct(A=@n8, B=@n9), B=DotProduct(A=@n8, B=@n10)), "A == B"=@n11, "A < B"=@n11)
  n13 = FunctionInput() {
   bUsePreviewValueAsDefault=true
   description=The size of the virtual plane
   inputName=UV Scale
   inputType=FunctionInput_Vector2
   sortPriority=3
   previewValue=<1 1 0 1>
  }
  FunctionOutput(None=Add(A=Divide(A=@n12, B=@n13)) {constB=0.5}) {description=The 2D UVs of the virtual plane coordinates., outputName=UVs}
  FunctionOutput(None=@n7) {description=The world position at each point on the virtual plane., outputName=World Position, sortPriority=1}
 ]
}
